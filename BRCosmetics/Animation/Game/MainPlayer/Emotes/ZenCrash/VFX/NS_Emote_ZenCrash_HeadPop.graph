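# Niagara burst effect for the head pop in the ZenCrash emote, throwing out dust and toon debris.
# Dust partitions two staggered sprite bursts around a ring, pushing them outward with drag and animated material parameters.
# ToonDebris fires three spawn-grouped mesh/sprite bursts that fall, follow source movement horizontally, take time-of-day tint and shrink, scaled for front end.
# FlashStar, SparksCPU, SpeedLines, ToonSmoke and ToonSpikes are present but disabled.
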NiagaraSystem /BRCosmetics/Animation/Game/MainPlayer/Emotes/ZenCrash/VFX/NS_Emote_ZenCrash_HeadPop
dependencies {
 usedDataInterfaces [
  &NiagaraDataInterfaceColorCurve
  &NiagaraDataInterfaceCurve
  &NiagaraDataInterfaceVector2DCurve
  &NiagaraDataInterfaceSpriteRendererInfo
  &NiagaraDataInterfaceMeshRendererInfo
 ]
 usedDynamicInputs [&SimulationPosition]
 usedModules [&SystemState]
}
emitters {
 Dust {
  emitterUpdateScript [
   &EmitterState(Life Cycle Mode=::NewEnumerator1, Loop Behavior=::NewEnumerator1, Loop Duration=1.5, UseLoopDelay=false)
   &SpawnBurst_Instantaneous(Spawn Count=7, Use Spawn Probability=false, UseLoopCountLimit=false)
   &SpawnBurst_Instantaneous(Spawn Count=7, Spawn Group=1, Spawn Time=0.01, Use Spawn Probability=false, UseLoopCountLimit=false) {module=SpawnBurst_Instantaneous001}
   /SetVariables(Emitter.SizeRad=25) {isSetParameters=true}
  ]
  particleSpawnScript [
   &PartitionParticles(
    Particles Per Partition=7
    Partition Count Mode:ENiagara_InfiniteLoopDuration=::NewEnumerator1
    Partition Mode:ENiagara_PartitionMode_v2=::NewEnumerator1
   )
   /SetVariables(Particles.DustColor:LinearColor=&RandomRangeLinearColor(Minimum=<rgba 0.4635 0.4212 0.4025 1>, Maximum=<rgba 0.3229 0.2978 0.2778 1>)) {isSetParameters=true}
   /SetVariables(
    Particles.SizeMax:NiagaraFloat=&MaskFloatBySpawnGroup(Masked Float=45, Passthrough Float=20)
    Particles.SizeMin:NiagaraFloat=&MaskFloatBySpawnGroup(Masked Float=65, Passthrough Float=40)
   ) {isSetParameters=true}
   /SetVariables(Particles.SizeCurve:NiagaraFloat=&Lerp_Float(
     A=$Particles.SizeMin
     B=$Particles.SizeMax
     Alpha=&FloatFromCurve(
      FloatCurve=DataInterface() {propertyValues {Curve {keys [
          {interpMode=RCIM_Cubic, tangentMode=RCTM_Break, time=0, value=0}
          {arriveTangentWeight=0.9217, interpMode=RCIM_Cubic, tangentMode=RCTM_Break, time=1, value=1}
         ]}}}
      CurveIndex=&RandomRangeFloat(Minimum=0, Maximum=1)
     )
    )) {isSetParameters=true}
   /SetVariables(Particles.InitSpriteRotRate=&UniformRangedFloat(Minimum=-1, Maximum=1)) {isSetParameters=true}
   &InitializeParticle(
    Color=$Particles.DustColor
    Lifetime=&Lerp_Float(
     A=&MaskFloatBySpawnGroup(Masked Float=0.6, Passthrough Float=0.2)
     B=&MaskFloatBySpawnGroup(Masked Float=1, Passthrough Float=0.4)
     Alpha=&FloatFromCurve(
      FloatCurve=DataInterface() {propertyValues {Curve {keys [
          {interpMode=RCIM_Cubic, tangentMode=RCTM_User, time=0, value=0}
          {arriveTangent=2.8717, interpMode=RCIM_Cubic, leaveTangent=2.8717, tangentMode=RCTM_User, time=1, value=1}
         ]}}}
      CurveIndex=$Particles.PartitionParticles.Normalized Index In Partition
     )
    )
    Sprite Rotation Mode=::NewEnumerator2
    Sprite Size Mode=::NewEnumerator3
    Uniform Sprite Size=$Particles.SizeCurve
    UsePositionOffset=false
    Write Material Random=true
   )
   &ShapeLocation(
    Cylinder Distribution=::NewEnumerator2
    Cylinder Height=0
    Cylinder Height Midpoint=0
    Cylinder Radius=5
    MaskSpawnGroup=false
    Offset=&MaskVectorBySpawnGroup(Masked Vector=<0 0 20>, Passthrough Vector=<0 0 25>)
    Offset Coordinate Space=::World
    Offset Mode=::NewEnumerator0
    Shape Primitive=::NewEnumerator1
    U Position=$Particles.PartitionParticles.Normalized Index In Partition
    UseOwnerScale=true
   )
   &AddVelocity(
    Coordinate Space=::World
    Velocity=&Multiply_VectorByFloat(
     Vector=$Particles.ShapeLocation.ShapeNormal
     Float=&MergeHelper_FloatSwitch(
      0=&RandomRangeFloat(Minimum=154, Maximum=245)
      1=&RandomRangeFloat(Minimum=35, Maximum=105)
      2=&RandomRangeFloat(Minimum=70, Maximum=140)
      3=?
      Index=&SpawnGroup()
      Num=::NewEnumerator1
     )
    )
   ) {module=AddVelocity001}
   &CameraOffset(Camera Offset Amount=&RandomRangeFloat(Minimum=0, Maximum=5))
  ]
  particleUpdateScript [
   &ParticleState()
   &Drag(Drag=4, Use Linear Drag=true, Use Rotational Drag=false)
   &GravityForce(Gravity=<0 0 -50>)
   &SolveForcesAndVelocity(Clamp Velocity=false, Limit Acceleration=false)
   &SpriteRotationRate(Rotation Rate=&Multiply_Float(
     A=$Particles.InitSpriteRotRate
     B=&FloatFromCurve(
      FloatCurve=DataInterface() {propertyValues {Curve {keys [
          {arriveTangent=-3.9021, interpMode=RCIM_Cubic, leaveTangent=-3.9021, leaveTangentWeight=1.3778, tangentMode=RCTM_User, time=0, value=1}
          {arriveTangentWeight=0.3436, interpMode=RCIM_Cubic, tangentMode=RCTM_User, time=0.75, value=0.1}
         ]}}}
      Scale Curve=50
     )
    ))
   &ScaleSpriteSize(
    Scale Sprite Size Mode=::NewEnumerator0
    Uniform Scale Factor=&FloatFromCurve(FloatCurve=DataInterface() {propertyValues {Curve {keys [
         {interpMode=RCIM_Cubic, leaveTangent=1.0601, tangentMode=RCTM_Break, time=0, value=0.2}
         {arriveTangent=0.5926, interpMode=RCIM_Cubic, tangentMode=RCTM_Break, time=0.95, value=1}
         <at 1 0>
        ]}}})
   )
   &DynamicMaterialParameters(
    Index 0 Param 1=&Add_Float(
     A=&RandomRangeFloat(Minimum=0, Maximum=1)
     B=&FloatFromCurve(
      FloatCurve=DataInterface() {propertyValues {Curve {keys [{tangentMode=RCTM_User, time=0, value=0}, {tangentMode=RCTM_User, time=1, value=1}]}}}
      CurveIndex=&Add_Float(A=&Multiply_Float(A=$Particles.Initial.MaterialRandom, B=1), B=$Particles.NormalizedAge)
      Scale Curve=0.1
     )
    )
    Index 0 Param 2=&FloatFromCurve(FloatCurve=DataInterface() {propertyValues {Curve {keys [<at 0 1>,<at 1 0>]}}})
    Index 0 Param 3=&FloatFromCurve(
     FloatCurve=DataInterface() {propertyValues {Curve {keys [
         {interpMode=RCIM_Cubic, leaveTangent=16.5652, leaveTangentWeight=0.4385, tangentMode=RCTM_Break, time=0, value=0}
         {arriveTangentWeight=0.0333, interpMode=RCIM_Cubic, leaveTangentWeight=0.9477, tangentMode=RCTM_Break, time=0.1, value=1}
         {arriveTangent=-1.7159, interpMode=RCIM_Cubic, leaveTangent=-1.7159, tangentMode=RCTM_User, time=0.3, value=0.2}
         {arriveTangent=0.2347, interpMode=RCIM_Cubic, leaveTangent=0.2723, tangentMode=RCTM_Break, time=0.6, value=0}
        ]}}}
     Scale Curve=0
    )
    Index 0 Param 4=&FloatFromCurve(FloatCurve=DataInterface() {propertyValues {Curve {keys [{interpMode=RCIM_Cubic, tangentMode=RCTM_Break, time=0.5, value=1}, {interpMode=RCIM_Cubic, tangentMode=RCTM_Break, time=1, value=0}]}}})
    Index 1 Param 1=&FloatFromCurve(FloatCurve=DataInterface() {propertyValues {Curve {keys [
         {interpMode=RCIM_Cubic, leaveTangentWeight=0.0833, tangentMode=RCTM_Break, time=0, value=0}
         {arriveTangentWeight=0.0873, interpMode=RCIM_Cubic, tangentMode=RCTM_Break, time=0.25, value=1}
        ]}}})
    Index 1 Param 2=&FloatFromCurve(
     FloatCurve=DataInterface() {propertyValues {Curve {keys [
         {interpMode=RCIM_Cubic, tangentMode=RCTM_Break, time=0, value=0.2}
         {interpMode=RCIM_Cubic, leaveTangent=0.9066, tangentMode=RCTM_Break, time=1, value=2}
        ]}}}
     Scale Curve=&RandomRangeFloat(Minimum=0.1, Maximum=0.2)
    )
    Index 1 Param 3=&FloatFromCurve(FloatCurve=DataInterface() {propertyValues {Curve {keys [
         {interpMode=RCIM_Cubic, leaveTangent=6.3917, tangentMode=RCTM_Break, time=0, value=0}
         {arriveTangent=1.7549, interpMode=RCIM_Cubic, leaveTangent=10.6767, tangentMode=RCTM_User, time=0.25, value=1}
        ]}}})
    Index 1 Param 4=&FloatFromCurve(
     FloatCurve=DataInterface() {propertyValues {Curve {keys [{arriveTangent=-0.136, interpMode=RCIM_Cubic, leaveTangent=-0.136, tangentMode=RCTM_User, time=0, value=1}, <at 1 0.95>]}}}
     Scale Curve=2
    )
    Index 2 Param 1=&FloatFromCurve(
     FloatCurve=DataInterface() {propertyValues {Curve {keys [
         {interpMode=RCIM_Cubic, leaveTangent=4.2485, tangentMode=RCTM_Break, time=0, value=0}
         {arriveTangent=1.0548, interpMode=RCIM_Cubic, leaveTangent=1.0548, tangentMode=RCTM_User, time=0.17, value=0.5715}
         <at 1 1>
        ]}}}
     Scale Curve=3
    )
    Index 2 Param 2=&FloatFromCurve(
     FloatCurve=DataInterface() {propertyValues {Curve {keys [{interpMode=RCIM_Cubic, tangentMode=RCTM_User, time=0, value=0}, {interpMode=RCIM_Cubic, tangentMode=RCTM_User, time=1, value=1}]}}}
     Scale Curve=0.3
    )
    Index 2 Param 3=&FloatFromCurve(
     FloatCurve=DataInterface() {propertyValues {Curve {keys [
         {interpMode=RCIM_Cubic, leaveTangent=-15.7052, leaveTangentWeight=0.1411, tangentMode=RCTM_Break, time=0, value=1}
         {interpMode=RCIM_Cubic, tangentMode=RCTM_Break, time=0.1, value=0}
         {arriveTangent=0.7532, interpMode=RCIM_Cubic, leaveTangent=0.7532, tangentMode=RCTM_User, time=0.7, value=0.5}
         {arriveTangent=2.358, arriveTangentWeight=1.057, interpMode=RCIM_Cubic, tangentMode=RCTM_Break, time=1, value=1}
        ]}}}
     Scale Curve=5
    )
    Index 2 Param 4=&RandomRangeFloat(Minimum=1.5, Maximum=1.75)
    Module.Param0WriteEnabled=true
    Module.Param0WriteEnabled1=true
    Module.Param0WriteEnabled2=true
    Module.Param1WriteEnabled=false
    Module.Param1WriteEnabled1=true
    Module.Param1WriteEnabled2=true
    Module.Param2WriteEnabled=true
    Module.Param2WriteEnabled1=true
    Module.Param2WriteEnabled2=true
    Module.Param3WriteEnabled=true
    Module.Param3WriteEnabled1=true
    Module.Param3WriteEnabled2=true
    Write Parameter Index 1=true
    Write Parameter Index 2=true
   )
   &ScaleColor(
    Scale Alpha=&FloatFromCurve(
     FloatCurve=DataInterface() {propertyValues {Curve {keys [
         {arriveTangent=6.442, interpMode=RCIM_Cubic, leaveTangent=6.442, tangentMode=RCTM_User, time=0, value=0}
         {interpMode=RCIM_Cubic, tangentMode=RCTM_User, time=0.1, value=1}
         <at 1 0>
        ]}}}
     Scale Curve=&RandomRangeFloat(Minimum=0.75, Maximum=1)
    )
    ScaleA=true
    ScaleRGB=false
   )
  ]
  rendererClasses [&NiagaraSpriteRendererProperties]
  renderers [
   {rendererClass=&NiagaraSpriteRendererProperties}
   {rendererClass=&NiagaraSpriteRendererProperties, rendererIndex=1}
  ]
 }
 FlashStar {
  bEnabled=false
  emitterUpdateScript [
   &EmitterState(
    Enable Distance Culling=true
    Max Distance Response=::KillImmediately
    Module.bUseMaxDistance=true
    Module.bUseMinDistance=false
    Scalability Mode=::NewEnumerator1
   )
   &SpawnBurst_Instantaneous(Use Spawn Probability=false, UseLoopCountLimit=false)
  ]
  particleSpawnScript [
   &InitializeParticle(
    Color Mode=::NewEnumerator0
    Lifetime=0.133
    Sprite Rotation Mode=::NewEnumerator2
    Sprite Size Mode=::NewEnumerator3
    Uniform Sprite Size=&Multiply_Float(
     A=&UniformRangedFloat(Minimum=100, Maximum=125)
     B=&FN_FrontEndFloatModifier(Gameplay=1, Front End=$User.FrontEndScale)
    )
    UsePositionOffset=false
    Write Material Random=true
   )
   &CameraOffset(Camera Offset Amount=10)
  ]
  particleUpdateScript [
   &ParticleState()
   &ScaleSpriteSize(Uniform Curve Sprite Scale=DataInterface() {propertyValues {Curve {keys [
        {arriveTangent=2.105, interpMode=RCIM_Cubic, leaveTangent=2.105, tangentMode=RCTM_User, time=0, value=0}
        {arriveTangent=0.0224, interpMode=RCIM_Cubic, leaveTangent=0.0224, tangentMode=RCTM_User, time=1, value=1}
       ]}}}) {module=ScaleSpriteSize001}
   &ScaleColor(
    Scale Alpha=&Multiply_Float(
     A=&FloatFromCurve(FloatCurve=DataInterface() {propertyValues {Curve {keys [
          {arriveTangent=9.0725, interpMode=RCIM_Cubic, leaveTangent=9.0725, tangentMode=RCTM_User, time=0, value=0}
          {arriveTangent=0.3975, interpMode=RCIM_Cubic, leaveTangent=0.3975, tangentMode=RCTM_User, time=0.0555, value=1}
          {arriveTangent=-0.0405, interpMode=RCIM_Cubic, leaveTangent=-0.0405, tangentMode=RCTM_User, time=1, value=0}
         ]}}})
     B=1
    )
    Scale RGB=&VectorFromFloat(Value=&FloatFromCurve(FloatCurve=DataInterface() {propertyValues {Curve {keys [
          <at 0 10 RCIM_Cubic>
          {arriveTangent=-9, interpMode=RCIM_Cubic, leaveTangent=-9, time=0.5, value=3}
          {tangentMode=RCTM_User, time=1, value=1}
         ]}}}))
    ScaleA=true
    ScaleRGB=true
   )
   /SetVariables(
    Particles.InnerColor:LinearColor=&MultiplyLinearColor(
     Linear Color=<rgba 33 19.5083 2.4107 1>
     Scale Factor=&MakeVector4_FromLinearColor(LinearColor=$Particles.Color)
    )
    Particles.InnerDynamicParameter:Vector4f=&MakeVector4_FromFloat(Float=&Lerp_Float(A=0, B=0.5, Alpha=$Particles.NormalizedAge))
    Particles.InnerSize:Vector2f=&Multiply_Vector2D_ByFloat(Float=0.5, Vector2D=$Particles.SpriteSize)
    Particles.MiddleColor:LinearColor=&MultiplyLinearColor(
     Linear Color=<rgba 10 2.2754 0.7305 1>
     Scale Factor=&MakeVector4_FromLinearColor(LinearColor=$Particles.Color)
    )
    Particles.MiddleDynamicParameter:Vector4f=&MakeVector4_FromFloat(Float=&Lerp_Float(A=0.25, B=0.75, Alpha=$Particles.NormalizedAge))
    Particles.MiddleSize:Vector2f=&Multiply_Vector2D_ByFloat(Float=0.75, Vector2D=$Particles.SpriteSize)
    Particles.OuterColor:LinearColor=&MultiplyLinearColor(Linear Color=<rgba 0 0 0 1>, Scale Factor=&MakeVector4_FromLinearColor(LinearColor=$Particles.Color))
    Particles.OuterDynamicParameter:Vector4f=&MakeVector4_FromFloat(Float=&Lerp_Float(A=0, B=1, Alpha=$Particles.NormalizedAge))
    Particles.OuterSize:Vector2f=&Multiply_Vector2D_ByFloat(Float=1.3, Vector2D=$Particles.SpriteSize)
    Particles.SubImageIndex=&RandomRangeFloat(Minimum=0, Maximum=4)
    Particles.SubImageIndexAlt:NiagaraFloat=&RandomRangeFloat(Minimum=0, Maximum=3)
   ) {isSetParameters=true}
  ]
  rendererClasses [&NiagaraSpriteRendererProperties]
  renderers [
   {rendererClass=&NiagaraSpriteRendererProperties}
   {rendererClass=&NiagaraSpriteRendererProperties, rendererIndex=1}
   {rendererClass=&NiagaraSpriteRendererProperties, rendererIndex=2}
  ]
 }
 SparksCPU {
  bEnabled=false
  emitterUpdateScript [
   &EmitterState(
    Enable Distance Culling=true
    Max Distance Response=::KillImmediately
    Module.bUseMaxDistance=true
    Module.bUseMinDistance=false
    Scalability Mode=::NewEnumerator1
   )
   &SpawnBurst_Instantaneous(Spawn Count=&UniformRangedInt(Minimum=5, Maximum=10), Use Spawn Probability=false)
   &SpawnRate(
    Spawn Probability=&FloatFromCurve(
     FloatCurve=DataInterface() {propertyValues {Curve {keys [{arriveTangent=-12.2872, interpMode=RCIM_Cubic, leaveTangent=-12.2872, tangentMode=RCTM_User, time=0, value=1}, <at 0.1 0>]}}}
     CurveIndex=$System.Age
    )
    SpawnRate=111
    Use Spawn Probability=true
   )
  ]
  particleSpawnScript [
   &InitializeParticle(
    Color=<rgba 3 3 3 1>
    Lifetime Max=0.45
    Lifetime Min=0.166
    Lifetime Mode=::NewEnumerator1
    Sprite Size Mode=::NewEnumerator3
    Uniform Sprite Size=&Multiply_Float(
     A=&MakeCustomFloatFromBool(Bool=$User.bIsFrontEnd, True Float=0.45, False Float=1)
     B=&RandomRangeFloat(Minimum=2.5, Maximum=7)
    )
    UsePositionOffset=false
    Write Material Random=true
   )
   &ShapeLocation(MaskSpawnGroup=false, Override Seed=false, Sphere Radius=9, Sphere Surface Distribution=1, UseOwnerScale=true)
   &AddVelocity(
    Module.Use Falloff Distance=false
    Override Seed=false
    Velocity Mode=::NewEnumerator1
    Velocity Speed=&FN_FrontEndFloatModifier(
     Gameplay=&RandomRangeFloat(Minimum=222, Maximum=666)
     Front End=&RandomRangeFloat(Minimum=150, Maximum=400)
    )
   )
   &DynamicMaterialParameters(
    Index 0 Param 1=3
    Index 0 Param 2=20
    Index 0 Param 3=1
    Module.Param0WriteEnabled=true
    Module.Param1WriteEnabled=true
    Module.Param2WriteEnabled=true
    Module.Param3WriteEnabled=true
   )
  ]
  particleUpdateScript [
   &ParticleState()
   &GravityForce(Gravity=<0 0 -666>)
   &Drag(Drag=&UniformRangedFloat(Minimum=2, Maximum=6), Use Linear Drag=true, Use Rotational Drag=false)
   &SolveForcesAndVelocity(Clamp Velocity=false, Limit Acceleration=false)
   &ScaleSpriteSize(Scale Factor=&Vector2DFromCurve(
     Vector2Curve=DataInterface() {propertyValues {
       XCurve {keys [
         {arriveTangent=5.9728, interpMode=RCIM_Cubic, leaveTangent=5.9728, tangentMode=RCTM_User, time=0, value=0.1}
         {arriveTangent=-0.1, interpMode=RCIM_Cubic, leaveTangent=-0.1, tangentMode=RCTM_User, time=0.18, value=0.8}
         {interpMode=RCIM_Cubic, tangentMode=RCTM_User, time=1, value=0}
        ]}
       YCurve {keys [
         {arriveTangent=6.2715, interpMode=RCIM_Cubic, leaveTangent=6.2715, tangentMode=RCTM_User, time=0, value=0.1}
         {arriveTangent=-0.1001, interpMode=RCIM_Cubic, leaveTangent=-0.1001, tangentMode=RCTM_User, time=0.18, value=0.8}
         {interpMode=RCIM_Cubic, tangentMode=RCTM_User, time=0.9988, value=0}
        ]}
      }}
     ScaleCurve=<1 1>
    ))
   &ScaleSpriteSizeBySpeed(Max Scale Factor=<0.5 8>, Min Scale Factor=<0.5 1>, Velocity Threshold=666)
   &ScaleColor(
    Linear Color Curve=DataInterface() {propertyValues {
      AlphaCurve {keys [<at 0.333 1>,<at 1 0>]}
      BlueCurve {keys [<at 0 0.24>,<at 0.166 0.2>,<at 0.333 0>,<at 0.75 0>]}
      GreenCurve {keys [<at 0 3.12>,<at 0.166 2.6>,<at 0.333 0.6667>,<at 0.75 0.15>]}
      RedCurve {keys [<at 0 6>,<at 0.166 5>,<at 0.333 4>,<at 0.75 3>]}
     }}
    Scale Mode=::NewEnumerator2
   )
  ]
  rendererClasses [&NiagaraSpriteRendererProperties]
  renderers [{rendererClass=&NiagaraSpriteRendererProperties}]
 }
 SpeedLines {
  bEnabled=false
  emitterUpdateScript [
   &EmitterState(
    Enable Distance Culling=true
    Max Distance Response=::KillImmediately
    Module.bUseMaxDistance=true
    Module.bUseMinDistance=false
    Scalability Mode=::NewEnumerator1
   )
   &SpawnBurst_Instantaneous(Spawn Count=1, Use Spawn Probability=false)
  ]
  particleSpawnScript [
   &InitializeParticle(
    Lifetime=0.26
    Mass Mode=::NewEnumerator1
    Sprite Rotation Mode=::NewEnumerator2
    Sprite Size=&Vector2DFromFloat(Value=&Multiply_Float(
      A=&RandomRangeFloat(Minimum=140, Maximum=180)
      B=&FN_FrontEndFloatModifier(Gameplay=1, Front End=$User.FrontEndScale)
     ))
    Sprite Size Mode=::NewEnumerator4
    UsePositionOffset=false
    Write Material Random=true
   ) {module=InitializeParticle001}
   &CameraOffset(Camera Offset Amount=10)
  ]
  particleUpdateScript [
   &ParticleState()
   &SolveForcesAndVelocity(Clamp Velocity=false, Limit Acceleration=false)
   &ScaleSpriteSize(Scale Factor=&Vector2DFromFloat(Value=&FloatFromCurve(FloatCurve=DataInterface() {propertyValues {Curve {keys [{arriveTangent=3.9086, interpMode=RCIM_Cubic, leaveTangent=3.9086, tangentMode=RCTM_User, time=0, value=0}, <at 1 1.4 RCIM_Cubic>]}}})))
   &DynamicMaterialParameters(
    Index 0 Param 1=&FloatFromCurve(FloatCurve=DataInterface() {propertyValues {Curve {keys [<at 0 1 RCIM_Cubic>, {arriveTangent=-2.742, interpMode=RCIM_Cubic, leaveTangent=-2.742, tangentMode=RCTM_User, time=1, value=0}]}}})
    Module.Param0WriteEnabled=true
    Module.Param1WriteEnabled=true
    Module.Param2WriteEnabled=true
    Module.Param3WriteEnabled=true
   )
   &ScaleColor(
    Scale Alpha=&FloatFromCurve(
     FloatCurve=DataInterface() {propertyValues {Curve {keys [
         {arriveTangent=-13.9062, interpMode=RCIM_Cubic, leaveTangent=-13.9062, tangentMode=RCTM_User, time=0, value=7}
         {arriveTangent=-7, interpMode=RCIM_Cubic, leaveTangent=-7, time=0.2488, value=3}
         <at 1 0 RCIM_Cubic>
        ]}}}
     Scale Curve=2
    )
    ScaleA=true
    ScaleRGB=true
   )
   &ScaleColor(
    Scale RGB=&VectorFromFloat(Value=&FNTimeOfDayScalar(Night Scale=0.6, Morning Scale=0.9, Evening Scale=0.9))
    ScaleA=false
    ScaleRGB=true
   ) {module=ScaleColor001}
  ]
  rendererClasses [&NiagaraSpriteRendererProperties]
  renderers [{rendererClass=&NiagaraSpriteRendererProperties}]
 }
 ToonDebris {
  emitterUpdateScript [
   &EmitterState(
    Enable Distance Culling=true
    Life Cycle Mode=::NewEnumerator1
    Loop Behavior=::NewEnumerator1
    Loop Duration=14.4333
    Max Distance Response=::KillImmediately
    Module.bUseMaxDistance=true
    Module.bUseMinDistance=false
    Scalability Mode=::NewEnumerator1
    UseLoopDelay=false
   )
   &SpawnRate(SpawnRate=100, Use Spawn Probability=false) {enabled=false}
   &SpawnBurst_Instantaneous(
    Spawn Count=&RandomRangeInt(Minimum=4, Maximum=8)
    Use Spawn Probability=false
    UseLoopCountLimit=false
   )
   &SpawnBurst_Instantaneous(
    Spawn Count=&RandomRangeInt(Minimum=10, Maximum=18)
    Spawn Group=1
    Spawn Time=0.05
    Use Spawn Probability=false
    UseLoopCountLimit=false
   ) {module=SpawnBurst_Instantaneous001}
   &SpawnBurst_Instantaneous(
    Spawn Count=&RandomRangeInt(Minimum=1, Maximum=2)
    Spawn Group=2
    Spawn Time=0.1
    Use Spawn Probability=false
    UseLoopCountLimit=false
   ) {module=SpawnBurst_Instantaneous002}
  ]
  particleSpawnScript [
   &InitializeParticle(
    Color=<rgba 1 0.4706 0.1961 1>
    Lifetime=&RandomRangeFloat(Minimum=0.9, Maximum=1)
    Mesh Scale Max=&MergeHelper_VectorSwitch(0=<0.5 0.5 0.7>, 1=<0.35 0.35 0.35>, 2=<0.6 0.6 1>, 3=?, Index=&SpawnGroup(), Num=::NewEnumerator1)
    Mesh Scale Min=&MergeHelper_VectorSwitch(0=<0.35 0.35 0.35>, 1=<0.1 0.1 0.1>, 2=<0.4 0.4 0.5>, 3=?, Index=&SpawnGroup(), Num=::NewEnumerator1)
    Mesh Scale Mode=::NewEnumerator2
    UsePositionOffset=false
    Write Material Random=true
   )
   &ShapeLocation(
    Cylinder Height=2
    Cylinder Radius=10
    MaskSpawnGroup=false
    Offset=&Lerp_Vector(A=<0 0 25>, B=<0 0 40>, Alpha=$Emitter.NormalizedLoopAge)
    Offset Coordinate Space=::World
    Offset Mode=::NewEnumerator0
    Override Seed=false
    Shape Primitive=::NewEnumerator1
    Surface Only=true
    Surface Only Band Thickness=1
    Use Endcaps In Surface Only Mode=false
    UseOwnerScale=true
   )
   &InitialMeshOrientation(Rotation=&RandomRangeVector(Minimum=<-1 -1 0>, Maximum=<1 1 1>), Use Rotation Vector=true)
   &AddVelocityFromPoint(
    Default Position:NiagaraPosition=$StackContext.SkeletalMeshLocation.SampledPosition
    Module.Use Falloff Distance=false
    Origin Offset=<0 0 -5>
    Origin Offset Coordinate Space=::World
    Velocity Strength=&RandomRangeFloat(Minimum=350, Maximum=500)
   ) {enabled=false}
   &AddVelocity(
    Coordinate Space=::World
    Velocity=&Multiply_VectorByFloat(
     Vector=$Particles.ShapeLocation.ShapeNormal
     Float=&MergeHelper_FloatSwitch(
      0=&RandomRangeFloat(Minimum=200, Maximum=400)
      1=&RandomRangeFloat(Minimum=200, Maximum=400)
      2=&RandomRangeFloat(Minimum=50, Maximum=100)
      3=?
      Index=&SpawnGroup()
      Num=::NewEnumerator1
     )
    )
   ) {module=AddVelocity001}
   &AddVelocity(Coordinate Space=::World, Velocity=<0 0 50>)
   &AddVelocity(
    Cone Angle=110
    Cone Axis=<0 0 1>
    Module.UseConeFalloff=false
    Module.UseVelDistribution=false
    Override Seed=false
    Velocity Mode=::NewEnumerator2
    Velocity Speed=&MergeHelper_FloatSwitch(
     0=0
     1=0
     2=&RandomRangeFloat(Minimum=100, Maximum=450)
     3=?
     Index=&SpawnGroup()
     Num=::NewEnumerator1
    )
   ) {enabled=false, module=AddVelocity002}
   &SubUVAnimation(
    End Frame Range Override=3
    Random Start Frame=true
    "Sprite Renderer "=DataInterface() {propertyValues {SpriteRenderer=&self:ToonDebris_0.NiagaraSpriteRendererProperties_0}}
    UseEndFrame=true
    UseStartFrame=true
   )
   /SetVariables(Particles.Scale=&Multiply_VectorByFloat(Vector=$Particles.Scale, Float=0.75)) {isSetParameters=true}
  ]
  particleUpdateScript [
   &GravityForce()
   &ScaleColor(
    Scale RGB=&MakeVectorFromLinearColorRGB(Color=&ColorFromCurve(
      ColorCurve=DataInterface() {propertyValues {
        AlphaCurve {keys [<at 0 1 RCIM_Cubic>]}
        BlueCurve {keys [
          <at 0.2162 0.7 RCIM_Cubic>
          <at 0.2812 0.6314>
          {arriveTangent=1.3318, interpMode=RCIM_Cubic, leaveTangent=1.3318, time=0.3527, value=1}
          {arriveTangent=-1.0217, interpMode=RCIM_Cubic, leaveTangent=-1.0217, time=0.558, value=1}
          {arriveTangent=-0.8311, interpMode=RCIM_Cubic, leaveTangent=-0.8311, time=0.8029, value=0.54}
          <at 0.919 0.7>
         ]}
        GreenCurve {keys [
          <at 0.2162 0.35 RCIM_Cubic>
          <at 0.2812 0.6667>
          {arriveTangent=1.2043, interpMode=RCIM_Cubic, leaveTangent=1.2043, time=0.3527, value=1}
          {arriveTangent=-0.8536, interpMode=RCIM_Cubic, leaveTangent=-0.8536, time=0.558, value=1}
          {arriveTangent=-1.8007, interpMode=RCIM_Cubic, leaveTangent=-1.8007, time=0.8029, value=0.6157}
          <at 0.919 0.35>
         ]}
        RedCurve {keys [
          <at 0.2162 0.28 RCIM_Cubic>
          <at 0.2812 0.902>
          {arriveTangent=0.3542, interpMode=RCIM_Cubic, leaveTangent=0.3542, time=0.3527, value=1}
          {arriveTangent=-0.2221, interpMode=RCIM_Cubic, leaveTangent=-0.2221, time=0.558, value=1}
          {arriveTangent=-1.9947, interpMode=RCIM_Cubic, leaveTangent=-1.9947, time=0.8029, value=0.9}
          <at 0.919 0.28>
         ]}
       }}
      CurveIndex=Linked() {linkedVariable {name=NPC.FortniteNPC.RemappedTimeOfDay_0-1, type=NiagaraFloat}}
     ))
    ScaleA=false
    ScaleRGB=true
   ) {module=ScaleColor001}
   &Drag(Drag=&RandomRangeFloat(Minimum=1.5, Maximum=2), Use Linear Drag=true, Use Rotational Drag=false)
   &MeshRotationForce(Rotation=&RandomRangeVector(Minimum=<40 10 0>, Maximum=<80 20 0>))
   &InheritSourceMovement(
    Applied Position Falloff Distance=1000
    Applied Position Scale=&Multiply_VectorByFloat(
     Vector=<1 1 0>
     Float=&FloatFromCurve(
      FloatCurve=DataInterface() {propertyValues {Curve {keys [
          {interpMode=RCIM_Cubic, tangentMode=RCTM_User, time=0.1865, value=1}
          {arriveTangent=-2.1905, interpMode=RCIM_Cubic, leaveTangent=-2.1905, tangentMode=RCTM_User, time=1, value=0}
         ]}}}
      CurveIndex=$Emitter.Age
     )
    )
    ApplyForce=false
    ApplyPosition=true
    ApplyPositionFalloff=false
   )
   &SolveForcesAndVelocity(Clamp Velocity=false, Limit Acceleration=false)
   &ParticleState()
   &ScaleMeshSize(Scale Factor=&VectorFromFloat(Value=&FloatFromCurve(
      FloatCurve=DataInterface() {propertyValues {Curve {keys [
          {tangentMode=RCTM_User, time=0, value=1}
          {arriveTangent=-0.0405, interpMode=RCIM_Cubic, leaveTangent=-0.0405, tangentMode=RCTM_User, time=0.5, value=1}
          {arriveTangent=-4.4728, interpMode=RCIM_Cubic, leaveTangent=-4.4728, tangentMode=RCTM_User, time=1, value=0}
         ]}}}
      Scale Curve=&FN_FrontEndFloatModifier(Gameplay=1, Front End=$User.FrontEndScale)
     )))
  ]
  rendererClasses [&NiagaraMeshRendererProperties, &NiagaraSpriteRendererProperties]
  renderers [
   {rendererClass=&NiagaraMeshRendererProperties, rendererIndex=0}
   {rendererClass=&NiagaraSpriteRendererProperties, rendererIndex=1}
  ]
 }
 ToonSmoke {
  bEnabled=false
  emitterUpdateScript [
   &EmitterState(
    Enable Distance Culling=true
    Max Distance Response=::KillImmediately
    Module.bUseMaxDistance=true
    Module.bUseMinDistance=false
    Scalability Mode=::NewEnumerator1
   )
   &SpawnBurst_Instantaneous(Spawn Time=0.03, Use Spawn Probability=false, UseLoopCountLimit=false)
  ]
  particleSpawnScript [
   &InitializeParticle(
    Color Maximum=<rgba 0.6609 0.6609 0.6609 1>
    Color Minimum=<rgba 0.913 0.913 0.913 1>
    Color Mode=::NewEnumerator2
    Lifetime Max=0.35
    Lifetime Min=0.3
    Lifetime Mode=::NewEnumerator1
    Sprite Rotation Mode=::NewEnumerator2
    Sprite Size Mode=::NewEnumerator3
    Sprite UV Mode=::NewEnumerator3
    Uniform Sprite Size=&Multiply_Float(
     A=&RandomRangeFloat(Minimum=90, Maximum=120)
     B=&FN_FrontEndFloatModifier(Gameplay=1, Front End=$User.FrontEndScale)
    )
    UsePositionOffset=false
    Write Material Random=true
   )
   &CameraOffset(Camera Offset Amount=&RandomRangeFloat(Minimum=20, Maximum=30))
  ]
  particleUpdateScript [
   &ParticleState()
   &SpriteRotationRate(Rotation Rate=&RandomRangeFloat(Minimum=-20, Maximum=20))
   &DynamicMaterialParameters(
    Index 0 Param 1=$Particles.NormalizedAge
    Index 0 Param 2=1
    Module.Param0WriteEnabled=true
    Module.Param1WriteEnabled=true
    Module.Param2WriteEnabled=true
    Module.Param3WriteEnabled=true
   )
   &ScaleColor(
    Scale Alpha=&FloatFromCurve(FloatCurve=DataInterface() {propertyValues {Curve {keys [
         {tangentMode=RCTM_User, time=0, value=1}
         {arriveTangent=-0.0405, interpMode=RCIM_Cubic, leaveTangent=-0.0405, tangentMode=RCTM_User, time=0.7009, value=1}
         {arriveTangent=-8.0089, interpMode=RCIM_Cubic, leaveTangent=-8.0089, tangentMode=RCTM_User, time=1, value=0}
        ]}}})
    Scale RGB=&MakeVectorFromLinearColorRGB(Color=&ColorFromCurve(ColorCurve=DataInterface() {propertyValues {
        AlphaCurve {keys [<at 0 1 RCIM_Cubic>,<at 1 1 RCIM_Cubic>]}
        BlueCurve {keys [<at 0.0012 3.8249>,<at 0.8202 0.0104>]}
        GreenCurve {keys [<at 0.0012 3.928>,<at 0.8202 0.0104>]}
        RedCurve {keys [<at 0.0012 4>,<at 0.8202 0.0104>]}
       }}))
    ScaleA=true
    ScaleRGB=true
   )
   &ScaleColor(
    Scale RGB=&MakeVectorFromLinearColorRGB(Color=&ColorFromCurve(
      ColorCurve=DataInterface() {propertyValues {
        AlphaCurve {keys [<at 0 1 RCIM_Cubic>]}
        BlueCurve {keys [
          <at 0.2162 0.7 RCIM_Cubic>
          <at 0.2812 0.6314>
          {arriveTangent=1.3318, interpMode=RCIM_Cubic, leaveTangent=1.3318, time=0.3527, value=1}
          {arriveTangent=-1.0217, interpMode=RCIM_Cubic, leaveTangent=-1.0217, time=0.558, value=1}
          {arriveTangent=-0.8311, interpMode=RCIM_Cubic, leaveTangent=-0.8311, time=0.8029, value=0.54}
          <at 0.919 0.7>
         ]}
        GreenCurve {keys [
          <at 0.2162 0.35 RCIM_Cubic>
          <at 0.2812 0.6667>
          {arriveTangent=1.2043, interpMode=RCIM_Cubic, leaveTangent=1.2043, time=0.3527, value=1}
          {arriveTangent=-0.8536, interpMode=RCIM_Cubic, leaveTangent=-0.8536, time=0.558, value=1}
          {arriveTangent=-1.8007, interpMode=RCIM_Cubic, leaveTangent=-1.8007, time=0.8029, value=0.6157}
          <at 0.919 0.35>
         ]}
        RedCurve {keys [
          <at 0.2162 0.28 RCIM_Cubic>
          <at 0.2812 0.902>
          {arriveTangent=0.3542, interpMode=RCIM_Cubic, leaveTangent=0.3542, time=0.3527, value=1}
          {arriveTangent=-0.2221, interpMode=RCIM_Cubic, leaveTangent=-0.2221, time=0.558, value=1}
          {arriveTangent=-1.9947, interpMode=RCIM_Cubic, leaveTangent=-1.9947, time=0.8029, value=0.9}
          <at 0.919 0.28>
         ]}
       }}
      CurveIndex=Linked() {linkedVariable {name=NPC.FortniteNPC.RemappedTimeOfDay_0-1, type=NiagaraFloat}}
     ))
    ScaleA=false
    ScaleRGB=true
   ) {module=ScaleColor001}
  ]
  rendererClasses [&NiagaraSpriteRendererProperties]
  renderers [{rendererClass=&NiagaraSpriteRendererProperties}]
 }
 ToonSpikes {
  bEnabled=false
  emitterUpdateScript [
   &EmitterState(
    Enable Distance Culling=true
    Max Distance Response=::KillImmediately
    Module.bUseMaxDistance=true
    Module.bUseMinDistance=false
    Scalability Mode=::NewEnumerator1
   )
   &SpawnRate(SpawnRate=20, Use Spawn Probability=false) {enabled=false}
   &SpawnBurst_Instantaneous(Spawn Count=8, Use Spawn Probability=false)
  ]
  particleSpawnScript [
   &InitializeParticle(
    Color Maximum=<rgba 0 0 0 1>
    Color Minimum=<rgba 0 0 0 0.7>
    Color Mode=::NewEnumerator2
    Lifetime Max=0.2
    Lifetime Min=0.1
    Lifetime Mode=::NewEnumerator1
    Sprite Rotation Mode=::NewEnumerator2
    Sprite Size Max=<10 45>
    Sprite Size Min=<5 33.5>
    Sprite Size Mode=::NewEnumerator2
    UsePositionOffset=false
    Write Material Random=true
   ) {module=InitializeParticle001}
   &AddVelocityFromPoint(Module.Use Falloff Distance=false, Velocity Strength=140)
  ]
  particleUpdateScript [
   &CameraOffset(Camera Offset Amount=10)
   &ParticleState()
   &SolveForcesAndVelocity(Clamp Velocity=false, Limit Acceleration=false)
   &ScaleSpriteSize(Scale Factor=&Vector2DFromCurve(
     Vector2Curve=DataInterface() {propertyValues {XCurve {keys [<at 0 0.778>,<at 1 1>]}, YCurve {keys [<at 0 0.03>,<at 1 1.5>]}}}
     ScaleCurve=&FN_FrontEndVector2Modifier(Front End:Vector2f=<0.4 0.4>, Gameplay:Vector2f=<1 1>)
    ))
   &ScaleColor(
    Scale RGB=&Multiply_Vector(
     A=&MakeVectorFromLinearColorRGB(Color=&ColorFromCurve(ColorCurve=DataInterface() {propertyValues {
         AlphaCurve {keys []}
         BlueCurve {keys [<at 0 0 RCIM_Cubic>,<at 0.5317 2 RCIM_Cubic>]}
         GreenCurve {keys [<at 0 1.2763 RCIM_Cubic>,<at 0.5317 0.5972 RCIM_Cubic>]}
         RedCurve {keys [<at 0 2 RCIM_Cubic>,<at 0.5317 0.0546 RCIM_Cubic>]}
        }}))
     B=&VectorFromFloat(Value=&FloatFromCurve(FloatCurve=DataInterface() {propertyValues {Curve {keys [<at 0 10>,<at 1 0>]}}}))
    )
    ScaleA=true
    ScaleRGB=true
   )
  ]
  rendererClasses [&NiagaraSpriteRendererProperties]
  renderers [{rendererClass=&NiagaraSpriteRendererProperties}]
 }
}
summary {
 systemName=NS_Emote_ZenCrash_HeadPop
 emitters [
  {bEnabled=false, emitterName=ToonSmoke, rendererClasses [&NiagaraSpriteRendererProperties]}
  {bEnabled=false, emitterName=SpeedLines, rendererClasses [&NiagaraSpriteRendererProperties]}
  {bEnabled=false, emitterName=ToonSpikes, rendererClasses [&NiagaraSpriteRendererProperties]}
  {emitterName=ToonDebris, rendererClasses [&NiagaraMeshRendererProperties, &NiagaraSpriteRendererProperties]}
  {bEnabled=false, emitterName=SparksCPU, rendererClasses [&NiagaraSpriteRendererProperties]}
  {bEnabled=false, emitterName=FlashStar, rendererClasses [&NiagaraSpriteRendererProperties]}
  {emitterName=Dust, rendererClasses [&NiagaraSpriteRendererProperties]}
 ]
 userVariables [
  User.bIsFrontEnd {type=NiagaraBool}
  User.bIsFrontEndPreview {type=NiagaraBool}
  User.BurstIndex {type=NiagaraFloat, defaultValue=0, description=""}
  User.DesertBoulder {type=NiagaraBool}
  User.FrontEndScale {type=NiagaraFloat, defaultValue=0.4, description=""}
 ]
}
